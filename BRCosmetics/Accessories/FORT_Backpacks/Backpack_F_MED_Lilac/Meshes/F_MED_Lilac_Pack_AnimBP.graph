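# Anim Blueprint for the F_MED_Lilac backpack that procedurally bends its dyn_vert bones in response to spine motion.
# Each update converts spine yaw deltas from reference pose into roll: spine_03 drives vert_01 and a counter-rotated vert_base, spine_04 drives vert_02, amplified and clamped.
# The AnimGraph applies these through chained additive Modify Bone nodes over the input pose.
# A crouch-weighted dyn_vert_MSTR node carries zero rotation.
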AnimBlueprint /BRCosmetics/Accessories/FORT_Backpacks/Backpack_F_MED_Lilac/Meshes/F_MED_Lilac_Pack_AnimBP
target_skeleton=&/F_MED_Lilac_Pack_Skeleton
graphs [
 AnimGraph {nodes [
   n1 = "Transform (Modify) Bone - Bone: dyn_vert_MSTR"(
    ComponentPose=Local To Component(LocalPose=Input Pose())
    Alpha="To Float (Boolean)"(InBool=Get bIsCrouching())
    Translation="0,0,0"
    Rotation="0,0,0"
    Scale="1,1,1"
   ) {
    rotationMode=BMM_Additive
    alphaBoolBlend {blendOption=Linear}
    alphaScaleBiasClamp {bInterpResult=true}
    boneToModify {boneName=dyn_vert_MSTR}
   }
   n2 = "Transform (Modify) Bone - Bone: dyn_vert_01"(
    ComponentPose="Transform (Modify) Bone - Bone: dyn_vert_base"(ComponentPose=@n1, Alpha=1, Translation="0,0,0", Rotation=Get Vert_Base_Rotate() -> (Vert_Base_Rotate:Rotator), Scale="1,1,1") {rotationMode=BMM_Additive, alphaBoolBlend {blendOption=Linear}, boneToModify {boneName=dyn_vert_base}}
    Alpha=1
    Translation="0,0,0"
    Rotation=Get Vert_01_Rotate() -> (Vert_01_Rotate:Rotator)
    Scale="1,1,1"
   ) {rotationMode=BMM_Additive, alphaBoolBlend {blendOption=Linear}, boneToModify {boneName=dyn_vert_01}}
   n3 = Get Vert_02_Rotate() -> (Vert_02_Rotate:Rotator)
   Output Pose(Result=Component To Local(ComponentPose="Transform (Modify) Bone - Bone: dyn_vert_02"(ComponentPose=@n2, Alpha=1, Translation="0,0,0", Rotation=@n3, Scale="1,1,1") {rotationMode=BMM_Additive, alphaBoolBlend {blendOption=Linear}, boneToModify {boneName=dyn_vert_02}})) {layerGroup=None}
  ]}
 EventGraph {nodes [
   Event BlueprintUpdateAnimation()
   n1 = GetOwnerSkeletalMeshComponent()
   n2 = BreakRotator(InRot=BreakTransform(InTransform=GetDeltaTransformFromRefPose(self=@n1, BoneName=spine_03, BaseName=spine_02)).Rotation)
   > n3 = Set Vert_01_Rotate(Vert_01_Rotate:Rotator=MakeRotator(Roll=float * float(A=@n2.Yaw, B=1) {is=title K2Node_PromotableOperator})) -> (Output_Get:Rotator)
   > Set Vert_Base_Rotate(Vert_Base_Rotate:Rotator="Scale Rotator (Float)"(A:Rotator=@n3, B:float=-0.6) -> (ReturnValue:Rotator=~) {is=title K2Node_PromotableOperator}) -> (Output_Get:Rotator)
   n4 = GetDeltaTransformFromRefPose(self=GetOwnerSkeletalMeshComponent(), BoneName=spine_04, BaseName=spine_03)
   n5 = float * float(A=BreakRotator(InRot=BreakTransform(InTransform=@n4).Rotation).Yaw, B=2.1)
   > Set Vert_02_Rotate(Vert_02_Rotate:Rotator=MakeRotator(Roll="Clamp (Float)"(Value=@n5, Min=-12, Max=5))) -> (Output_Get:Rotator)
   TryGetPawnOwner()
  ]}
]
variables [
 Vert_01_Rotate {type=Rotator}
 Vert_02_Rotate {type=Rotator}
 Vert_Base_Rotate {type=Rotator}
]
